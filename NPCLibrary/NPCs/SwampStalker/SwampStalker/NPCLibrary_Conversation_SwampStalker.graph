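# Conversation database for the Swamp Stalker NPC, defining its greeting and the shared dialogue branches it offers.
# The entry point leads to one Speech node whose default line competes with contextual lines for returning players, first meetings, high Bars and a Merman outfit, prioritized by curve rows.
# Link nodes then hand off by entry tag to NPC library quests, basic services and the Season 17 Mindwipe narrative.
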ConversationDatabase /NPCLibrary/NPCs/SwampStalker/SwampStalker/NPCLibrary_Conversation_SwampStalker
entry_points [{tag=Conversation.Entry.SwampStalker, destinations [#1]}]
nodes [
 ConversationEntryPointNode_7 {
  id=#1
  class=ConversationEntryPointNode
  outputs [#2]
  properties {entryTag {tagName=Conversation.Entry.SwampStalker}}
 }
 Speech_C_4 {
  id=#2
  class=Speech_C
  outputs [#3, #4, #5]
  properties {
   nodeName=""
   generalConfig {
    defaultMessage="Stick to the shadows. The swampy, swampy shadows."
    contextualMessages [
     {
      message=Glad to have you back. We all get lost in the muck sometimes.
      contextRequirements [{requirement {class=FortControllerRequirement_CheckTimeSinceLastPlayedMatch}}]
      priority {curve {rowName=Tandem.Speech.Priority.WhileSincePlayed}}
     }
     {
      message="Hope you brought bug repellant. Where we're going, you'll need it."
      contextRequirements [{requirement {class=FortControllerRequirement_IsFirstEverConversationWithNPC}}]
      priority {curve {rowName=Tandem.Speech.Priority.FirstEverInteraction}}
     }
     {
      message="Whoa, did you dig up all those Bars in the swamp?"
      contextRequirements [{requirement {class=FortControllerRequirement_HasBars, properties {thresholdTestConfig {requirement=GreaterThanOrEqual}}}}]
      priority {curve {rowName=Tandem.Speech.Priority.HighTotalBars}}
     }
     {
      message=So... you doing anything later?
      contextRequirements [{requirement {
         class=FortControllerRequirement_HasCID
         properties {allowedCIds [/BRCosmetics/Athena/Items/Cosmetics/Characters/CID_130_Athena_Commando_M_Merman]}
        }}]
      priority {curve {rowName=Tandem.Speech.Priority.OutfitReaction}}
     }
    ]
   }
  }
 }
 ConversationLinkNode_2 {
  id=#5
  class=ConversationLinkNode
  properties {remoteEntryTag {tagName=Conversation.Entry.Season17.Narrative.Mindwipe}}
 }
 ConversationLinkNode_0 {id=#4, class=ConversationLinkNode, properties {remoteEntryTag {tagName=Conversation.Entry.BasicServices}}}
 ConversationLinkNode_1 {id=#3, class=ConversationLinkNode, properties {remoteEntryTag {tagName=Conversation.Entry.NPCLibrary.Quests}}}
]
